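; Weapon Blueprint for the ascended Mist Raven flamethrower, playing an equip animation on the pawn holding it.
; When the weapon is equipped and attached, it runs the parent logic, casts the instigator to a Fort player pawn and plays the MistRaven equip montage on that pawn's character mesh.
; Tick has no logic, and the construction script only calls the parent.
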(event FortWeapon|EventOnEquippedAndAttached
  (|CallParentFunction)
  (bind _asfort_player_pawn (Utilities|Casting|CastToFortPlayerPawn (Game|GetInstigator)))
  (Animation|Montage|PlayMontage (Variables|/Script/Engine.Character|GetMesh _asfort_player_pawn) "/MistRaven/Animation/MistRaven_Equip_M.MistRaven_Equip_M" 1.0 0.0 "None" false))

(event EventTick
  (params (DeltaSeconds :type Float%20%28single-precision%29)))

(fn ConstructionScript
  (|CallParentFunction))
