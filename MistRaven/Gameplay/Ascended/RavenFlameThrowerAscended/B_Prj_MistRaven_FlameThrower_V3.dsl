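; MistRaven flamethrower projectile that burns targets inside a sphere growing over its lifetime.
; Unless native optimizations are enabled, a fast looping timer sphere-traces at its current location, filtering hits by line of sight to instigator or camera.
; Valid hits get hit cues and gameplay effects, then join an ignore list.
; The instigator's AllAscendedEquipped tag selects Ascended cues and, on authority, the Ascended victim effect.
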(event Collision|EventActorBeginOverlap
  (params (OtherActor :type Actor%20Object%20Reference))
  (|CallParentFunction OtherActor))

(event EventBeginPlay
  (Utilities|FlowControl|Sequence
    (:then_0
      (bind _returnvalue (GameplayTags|HasMatchingGameplayTag(Message) (Variables|Actor|GetInstigator) "(TagName=\"MistRaven.AllAscendedEquipped\")"))
      ;[1 0 0 0.5] VFX: Switching GCN when Ascended
      (if _returnvalue
        (Variables|Default|SetHit%20Gameplay%20Cue%20Tag (Variables|Default|GetHit%20Gameplay%20Cue%20Tag_Asc))
        (Variables|Default|SetHit%20Soft%20Gameplay%20Cue%20Tag (Variables|Default|GetHit%20Soft%20Gameplay%20Cue%20Tag_Asc))
        (else
          _))
      (|CallParentFunction)
      (Variables|Default|SetStartingTime (Utilities|Time|GetTimeSeconds))
      (Variables|Default|SetStarting%20Location (Projectile|GetCurrentLocation self))
      (Variables|Default|SetCameraLocation (Camera|GetCameraLocation (Variables|/Script/Engine.PlayerController|GetPlayerCameraManager (Fortnite|GetFortPlayerControllerfromActor (Variables|Actor|GetInstigator)))))
      (if (not (ScalableFloat|GetValueasBool (Variables|Default|GetEnableNativeOptimizations)))
        (Utilities|Time|SetTimerbyEvent (EventDispatchers|CreateEvent :!event Custom|SlowerTick) 0.01 true true) ; Only use the timer if we are not using the optimization.
        (Utilities|Array|Clear (Variables|Default|GetBpActorsToIgnore)) ; speculative clear just in case\nsome sort of pooling scheme\nis at play
        (Utilities|Array|Add (Variables|Default|GetBpActorsToIgnore) (Game|GetInstigator))))
    (:then_1
      (if (ScalableFloat|GetValueasBool (Variables|Default|GetRemoveExtraCompsOptimization))
        (Components|DestroyComponent (Variables|FortProjectileSniperAthena|GetBulletTrailParticleComponent))))
    (:then_2
      (if (> (ScalableFloat|GetValueatLevel (Variables|Default|GetDebugging)) 0)
        (bind _returnvalue_2 (Development|GetConsoleVariableFloatValue "MistRaven.Projectile.DebugDuration"))
        (Variables|Default|SetDebugDrawTime _returnvalue_2)))))

(event Custom|SlowerTick
  (bind _self self)
  (bind _returnvalue (Projectile|GetCurrentLocation _self))
  (bind _returnvalue_2 (Utilities|Time|GetTimeSeconds))
  (bind _instigator (Variables|Actor|GetInstigator))
  (bind _returnvalue_3 (Ability|GetAbilitySystemComponent _instigator))
  (bind _instigator_2 (Variables|Actor|GetInstigator))
  (bind _returnvalue_4 (Ability|GetAbilitySystemComponent _instigator_2))
  (bind _startingradius (Variables|Default|GetStartingRadius))
  (bind _increaseradiuspersec (Variables|Default|GetIncreaseRadiusPerSec))
  (bind _startingtime (Variables|Default|GetStartingTime))
  (bind _returnvalue_5 (- _returnvalue_2 _startingtime))
  (bind _returnvalue_6 (* _returnvalue_5 _increaseradiuspersec))
  (bind _returnvalue_7 (+ _returnvalue_6 _startingradius))
  (if (> (ScalableFloat|GetValueatLevel (Variables|Default|GetDebugging)) 0)
    (Rendering|Debug|DrawDebugSphere _returnvalue _returnvalue_7 7 "(R=1.000000,G=0.183615,B=0.229000,A=1.000000)" (Variables|Default|GetDebugDrawTime) 1.0)
    (else
      _))
  (bind _outhits (Collision|MultiSphereTraceByChannel _returnvalue _returnvalue _returnvalue_7 "TraceTypeQuery6" true (Variables|Default|GetBpActorsToIgnore) "ForDuration" true "(R=1.000000,G=0.000000,B=0.000000,A=1.000000)" "(R=0.000000,G=1.000000,B=0.000000,A=1.000000)" (Variables|Default|GetDebugDrawTime)))
  (for _array_element _outhits
    (bind _returnvalue_8 (GameplayTags|HasMatchingGameplayTag(Message) (.HitActor (Collision|BreakHitResult _array_element)) "(TagName=\"MistRaven.Flamethrower.Victim\")"))
    (if (not _returnvalue_8)
      (bind _in_line_of_sight (Class|MissionBlueprintFunctionLibrary|CheckBasicVectortoActorLineOfSight (.ImpactPoint (Collision|BreakHitResult _array_element)) (Variables|Actor|GetInstigator) "0, 0, 0" "TraceTypeQuery8" "None" 2.0 (Utilities|Array|MakeArray (.HitActor (Collision|BreakHitResult _array_element))) true false))
      (bind (_outhit _returnvalue_9) (Collision|LineTraceByChannel (.ImpactPoint (Collision|BreakHitResult _array_element)) (Variables|Default|GetCameraLocation) "TraceTypeQuery8" true (Utilities|Array|MakeArray (.HitActor (Collision|BreakHitResult _array_element))) "None" true "(R=1.000000,G=0.000000,B=0.000000,A=1.000000)" "(R=0.000000,G=1.000000,B=0.000000,A=1.000000)" 2.0))
      (if (or _in_line_of_sight (not _returnvalue_9))
        (bind _returnvalue_10 (GameplayTags|HasMatchingGameplayTag(Message) (.HitActor (Collision|BreakHitResult _array_element)) "(TagName=\"MistRaven.Flamethrower.Victim\")"))
        (if _returnvalue_10
          _
          (elif (Variables|/Script/Engine.Actor|GetbCanBeDamaged (.HitActor (Collision|BreakHitResult _array_element)))
            (bind _returnvalue_11 (CallFunction|GetOwnerWeapon))
            (bind _outcueparameters (GameplayCue|BuildCueParametersfromHitResult _returnvalue_11 _array_element))
            (GameplayCue|ExecuteGameplayCueOnActor(Burst) (.HitActor (Collision|BreakHitResult _array_element)) (Variables|Default|GetHit%20Gameplay%20Cue%20Tag) _outcueparameters)
            (else
              (bind _returnvalue_12 (CallFunction|GetOwnerWeapon))
              (bind _outcueparameters_2 (GameplayCue|BuildCueParametersfromHitResult _returnvalue_12 _array_element))
              (GameplayCue|ExecuteGameplayCueOnActor(Burst) (.HitActor (Collision|BreakHitResult _array_element)) (Variables|Default|GetHit%20Soft%20Gameplay%20Cue%20Tag) _outcueparameters_2))))
        (bind _returnvalue_13 (GameplayTags|HasMatchingGameplayTag(Message) (Variables|Actor|GetInstigator) "(TagName=\"MistRaven.AllAscendedEquipped\")"))
        (if (and _returnvalue_13 (Networking|HasAuthority))
          (bind _output_get (Variables|Default|SetGameplayEffectSpec (GameplayEffects|MakeOutgoingSpec _returnvalue_4 "/MistRaven/Gameplay/Ascended/RavenFlameThrowerAscended/GE_MistRaven_FlameThrower_Victim_Ascended.GE_MistRaven_FlameThrower_Victim_Ascended_C" 0.0 (GameplayEffects|MakeEffectContext _returnvalue_4))))
          (bind _returnvalue_14 (Ability|GameplayEffect|GetEffectContext _output_get))
          (Ability|EffectContext|AddHitResult _returnvalue_14 _array_element true)
          (GameplayEffects|ApplyGameplayEffectSpecToTarget _returnvalue_4 (Variables|Default|GetGameplayEffectSpec) (Ability|GetAbilitySystemComponent (.HitActor (Collision|BreakHitResult _array_element))))
          (else
            (bind _output_get_2 (Variables|Default|SetGameplayEffectSpec (GameplayEffects|MakeOutgoingSpec _returnvalue_3 "/MistRaven/Gameplay/Ascended/RavenFlameThrowerAscended/GE_MistRaven_FlameThrower_Victim.GE_MistRaven_FlameThrower_Victim_C" 0.0 (GameplayEffects|MakeEffectContext _returnvalue_3))))
            (bind _returnvalue_15 (Ability|GameplayEffect|GetEffectContext _output_get_2))
            (Ability|EffectContext|AddHitResult _returnvalue_15 _array_element true)
            (GameplayEffects|ApplyGameplayEffectSpecToTarget _returnvalue_3 (Variables|Default|GetGameplayEffectSpec) (Ability|GetAbilitySystemComponent (.HitActor (Collision|BreakHitResult _array_element))))))
        (Fortnite|ApplyGameplayEffectsToActor (.HitActor (Collision|BreakHitResult _array_element)) (Variables|Default|GetEffects%20to%20Apply_Delay) (Transformation|GetActorLocation) false 0 (Ability|GetAbilitySystemComponent (Variables|Actor|GetInstigator)) (GameplayEffects|MakeEffectContext (Ability|GetAbilitySystemComponent (Variables|Actor|GetInstigator))))
        (Utilities|Array|Add (Variables|Default|GetBpActorsToIgnore) (.HitActor (Collision|BreakHitResult _array_element)))
        (if (> (ScalableFloat|GetValueatLevel (Variables|Default|GetDebugging)) 0)
          (Development|PrintString "MISTRAVEN APPLYING DAMAGE BP"))))))

(event EventEndPlay
  (params (EndPlayReason :type EEndPlayReason%20Enum :default "Destroyed"))
  (|CallParentFunction EndPlayReason)
  ; Reset starting time in case of pooling
  (Variables|Default|SetStartingTime (Utilities|Time|GetTimeSeconds)))

(fn CheckLOS
  (params (HitActor :type Actor%20Object%20Reference) (OwningActor :type Actor%20Object%20Reference))
  (bind _starting_location (Variables|Default|GetStarting%20Location))
  (bind _instigator (Variables|/Script/Engine.Actor|GetInstigator OwningActor))
  (bind _returnvalue (Pawn|GetController _instigator))
  (bind _returnvalue_2 (Controller|LineOfSightTo _returnvalue HitActor _starting_location false))
  (Development|PrintString (Utilities|String|ToString(Boolean) _returnvalue_2))
  (return _returnvalue_2))

(fn GetEffectContext
  (return 0))

(fn ConstructionScript
  (|CallParentFunction))
