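; Gameplay ability for the MistRaven ascended flamethrower, firing a projectile stream while primary fire is held.
; A montage notify starts the looping cue and a timer whose server tick drains ammo via a remainder counter and spawns pause-gated projectiles.
; Projectiles aim with optional random spread from a stance-based offset, trace-corrected against nearby blockers.
; Activation layers camera modes and a movement effect; ending clears state, then plays the outro, reloads or removes depleted items.
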(event Ability|EventActivateAbility
  (bind _cueflamethrowertagloop (Variables|Default|GetCueFlamethrowerTagLoop))
  (Variables|Default|SetTransitionToReload false)
  (Variables|Default|SetEnded false)
  (bind _asfort_player_pawn (Utilities|Casting|CastToFortPlayerPawn (Ability|GetAvatarActorfromActorInfo)))
  ; Fine tuning camera movement based on the strength of the weapon we are firing
  (Variables|Default|SetAs%20Fort%20Player%20Pawn _asfort_player_pawn)
  (if (FortWeapon|IsTargeting (Ability|GetCurrentSourceWeapon))
    (Variables|Default|SetCameraTimer 0.01)
    (else
      (Utilities|Casting|CastToB_MistRaven_CombinationAscended (Weapon|GetCurrentWeapon _asfort_player_pawn)
        (:then
          (Camera|PushCameraMode (Variables|Default|GetCamera%20Mode%20Ascended%20Faster))
          (Variables|Default|SetCameraTimer 0.1))
        (:CastFailed
          (Camera|PushCameraMode (Variables|Default|GetCamera%20Mode%20Basic))
          (Variables|Default|SetCameraTimer 0.15)))))
  (Utilities|Time|SetTimerbyEvent (delegate Custom|2ndCamera) (Variables|Default|GetCameraTimer))
  (bind _asb_mist_raven_weapon_ascended (Utilities|Casting|CastToB_MistRaven_WeaponAscended (Weapon|GetCurrentWeapon (Variables|Default|GetAs%20Fort%20Player%20Pawn))))
  (Variables|Default|SetAs%20B%20Mist%20Raven%20Weapon%20Ascended _asb_mist_raven_weapon_ascended)
  (GameplayCue|ExecuteGameplayCue (Variables|Default|GetAs%20Fort%20Player%20Pawn) (Variables|Default|GetFX_Initial_Cue))
  (bind _asynctaskproxy (Ability|Tasks|WaitGameplayEvent "(TagName=\"Event.Primary.Released\")" 0 true)
    (:then
      (Variables|Default|SetEndAbilityEvent _asynctaskproxy)
      (bind _returnvalue (Ability|ApplyGameplayEffectToOwner "/MistRaven/Gameplay/Ascended/RavenFlameThrowerAscended/GE_MistRaven_FlameThrower_MovementSpeed.GE_MistRaven_FlameThrower_MovementSpeed_C"))
      (Variables|Default|SetMovementspeedEffect _returnvalue)
      (bind (_asynctaskproxy_2 _notifytag) (Ability|Tasks|PlayMontageAndWaitNotify "Flamethrower" (Variables|Default|GetBeamMontageToPlay) 1.0 "None" false)
        (:then
          (Variables|Default|SetAnimMontage _asynctaskproxy_2))
        (:OnCompleted
          _)
        (:OnBlendOut
          _)
        (:OnInterrupted
          _)
        (:OnCancelled
          _)
        (:OnNotifyBegin
          (if (== _notifytag "(TagName=\"MistRaven.Flamethrower.MontageNotifyFire\")")
            (bind _returnvalue_2 (Utilities|Time|SetTimerbyEvent (delegate Custom|Tick) (ScalableFloat|GetValueatLevel (Variables|Default|GetFlamethrowerTickTimer)) true true))
            (Variables|Default|SetSuctionTargeting_TH _returnvalue_2)
            (bind _returnvalue_3 (Ability|GameplayCue|IsGameplayCueActive (CallFunction|GetAvatarasFortPlayerPawn) _cueflamethrowertagloop))
            (if (not _returnvalue_3)
              (Ability|AddGameplayCueWithParamsToOwner _cueflamethrowertagloop (Ability|GameplayCue|MakeGameplayCueParameters 1e+04 1e+04) false)))))
      (Ability|EndAbility))
    (:EventReceived
      (Ability|EndAbility))))

(event Custom|2ndCamera
  ; Fine tuning camera movement based on the strength of the weapon we are firing
  (Camera|PopCameraMode)
  (Camera|PushCameraMode (Variables|Default|GetCamera%20Mode%20Tether%202)))

(event Custom|Tick
  (bind _returnvalue (Ability|GetCurrentSourceWeapon))
  (bind _as_b_mist_raven_weapon_ascended (Variables|Default|GetAs%20B%20Mist%20Raven%20Weapon%20Ascended))
  (Ability|HasAuthority
    (:True
      (if (> (FortWeapon|GetMagazineAmmoCount _returnvalue) 0)
        (Variables|/MistRaven/Gameplay/Ascended/B_MistRaven_WeaponAscended.B_MistRaven_WeaponAscended_C|SetFlamethrowerAmmoRemainder (+ (Variables|/MistRaven/Gameplay/Ascended/B_MistRaven_WeaponAscended.B_MistRaven_WeaponAscended_C|GetFlamethrowerAmmoRemainder _as_b_mist_raven_weapon_ascended) 1.0) _as_b_mist_raven_weapon_ascended :!label _label)
        (if (> (Variables|/MistRaven/Gameplay/Ascended/B_MistRaven_WeaponAscended.B_MistRaven_WeaponAscended_C|GetFlamethrowerAmmoRemainder (Variables|Default|GetAs%20B%20Mist%20Raven%20Weapon%20Ascended)) (Variables|Default|GetFlamethrowerRemainderMax))
          (FortWeapon|ModAmmo _returnvalue -1)
          (Variables|/MistRaven/Gameplay/Ascended/B_MistRaven_WeaponAscended.B_MistRaven_WeaponAscended_C|SetFlamethrowerAmmoRemainder 0.0 _as_b_mist_raven_weapon_ascended))
        (bind _returnvalue_2 (GameplayTags|HasMatchingGameplayTag(Message) (Ability|GetAvatarActorfromActorInfo) "(TagName=\"MistRaven.Flamethrower.AttackerPauseProjectile\")"))
        (if (not _returnvalue_2)
          (Ability|ApplyGameplayEffectToOwner "/MistRaven/Gameplay/Ascended/RavenFlameThrowerAscended/GE_MistRaven_FlameThrower_Attacker_Pause.GE_MistRaven_FlameThrower_Attacker_Pause_C")
          (bind _returnvalue_3 (Default|CalcProjectileDirection (Default|GetProjectileSpawnLocation)))
          (Projectile|SpawnProjectilewithParams "/MistRaven/Gameplay/Ascended/RavenFlameThrowerAscended/B_Prj_MistRaven_FlameThrower_V3.B_Prj_MistRaven_FlameThrower_V3_C" (Ability|GetAvatarActorfromActorInfo) (Utilities|Struct|MakeSpawnProjectileParams (Default|GetProjectileSpawnLocation) "0.000000,0.000000,0.000000" _returnvalue_3 (Ability|MakeGameplayEffectContainerSpec (Variables|Default|GetOnHitApplicationTag) 0) "(TargetSelection=(List=,bStopAtFirstSuccess=True,bKeepCheckingListOnIndestructibleHit=False,AbilityTargetBlockedGameplayTags=(GameplayTags=),bUseWeaponRanges=True,bUseMaxYawAngleToTarget=False,MaxYawAngleToTarget=0.000000),TargetGameplayEffectSpecs=,OwnerGameplayEffectSpecs=,ActivationCues=(GameplayTags=),ImpactCues=(GameplayTags=),ImpactNoiseRange=0.000000,FlyByNoiseRange=0.000000,bOverrideChargeMagnitude=False,ChargeMagnitudeOverrideValue=0.000000)" false 0 5e+03 1e+05 0.0))
          (if (== (FortWeapon|GetRemainingAmmo (Ability|GetCurrentSourceWeapon)) 0)
            (Variables|Default|SetDeleteLater true)))
        (elif (== (FortWeapon|GetRemainingAmmo (Ability|GetCurrentSourceWeapon)) 0)
          (Variables|Default|SetDeleteLater true)))
      (Ability|EndAbility))
    (:False
      (if (> (FortWeapon|GetLocalMagazineAmmoCount _returnvalue) 0)
        _)))
  (goto _label))

(event Custom|RPC_PlayExitAnim
  (Ability|Tasks|PlayMontageAndWaitNotify "Flamethrower_Outro" "/MistRaven/Animation/MistRaven_Drain_Outro_M.MistRaven_Drain_Outro_M" 1.0 "None" false))

(event Ability|EventOnEndAbility
  (params (bWasCancelled :type Boolean))
  (bind _returnvalue (CallFunction|GetAvatarasFortPlayerPawn))
  (bind _returnvalue_2 (Fortnite|GetFortPlayerControllerfromActor _returnvalue))
  (Ability|RemoveGameplayCueFromOwner (Variables|Default|GetCueFlamethrowerTagLoop))
  (Camera|PopCameraMode)
  (Utilities|Time|ClearandInvalidateTimerbyHandle (Variables|Default|GetSuctionTargeting_TH))
  (Components|Activation|Deactivate (Variables|Default|GetFlameThrowerVisuals))
  (GameplayTasks|EndTask (Variables|Default|GetEndAbilityEvent))
  (Ability|RemoveGameplayEffectFromOwnerWithHandle (Variables|Default|GetMovementspeedEffect))
  (if (Variables|Default|GetDeleteLater)
    (Ability|HasAuthority
      (:True
        (Inventory|RemoveItemfromInventoryOwner _returnvalue_2 "/MistRaven/Gameplay/Ascended/RavenFlameThrowerAscended/WID_MistRave_FlameThrowerAscended.WID_MistRave_FlameThrowerAscended")
        (Inventory|RemoveItemfromInventoryOwner _returnvalue_2 "/MistRaven/Gameplay/Ascended/CombinedAscended/WID_MistRaven_CombinedAscended.WID_MistRaven_CombinedAscended")))
    (elif (> (FortWeapon|GetMagazineAmmoCount (Ability|GetCurrentSourceWeapon)) 0)
      (bind _returnvalue_3 (GameplayTags|HasMatchingGameplayTag(Message) (Ability|GetAvatarActorfromActorInfo) "(TagName=\"MistRaven.GrimStrut.Active\")"))
      (if (not _returnvalue_3)
        (Ability|Tasks|PlayMontageAndWaitNotify "Flamethrower_Outro" "/MistRaven/Animation/MistRaven_Drain_Outro_M.MistRaven_Drain_Outro_M" 1.0 "None" false))
      (else
        (Ability|HasAuthority
          (:True
            (FortWeapon|Reload (Ability|GetCurrentSourceWeapon))))))))

(fn CalcProjectileDirection
  (params (Start :type Vector%20%28by%20ref%29) (IgnoreSpread :type Boolean))
  (bind _returnvalue (Ability|GetAvatarActorfromActorInfo))
  (bind _returnvalue_2 (Fortnite|GetFortPlayerControllerfromActor _returnvalue))
  (bind _randomizeprojectilesoffsetz (Variables|Default|GetRandomizeProjectilesOffsetZ))
  (bind _returnvalue_3 (ScalableFloat|GetValueasDouble _randomizeprojectilesoffsetz))
  (bind _returnvalue_4 (* -1.0 _returnvalue_3))
  (bind _returnvalue_5 (Math|Random|RandomFloatinRange _returnvalue_4 _returnvalue_3))
  (bind _randomizeprojectilesoffsety (Variables|Default|GetRandomizeProjectilesOffsetY))
  (bind _returnvalue_6 (ScalableFloat|GetValueasDouble _randomizeprojectilesoffsety))
  (bind _returnvalue_7 (* -1.0 _returnvalue_6))
  (bind _returnvalue_8 (Math|Random|RandomFloatinRange _returnvalue_7 _returnvalue_6))
  (bind _randomizeprojectilesoffsetx (Variables|Default|GetRandomizeProjectilesOffsetX))
  (bind _returnvalue_9 (ScalableFloat|GetValueasDouble _randomizeprojectilesoffsetx))
  (bind _returnvalue_10 (* -1.0 _returnvalue_9))
  (bind _returnvalue_11 (Math|Random|RandomFloatinRange _returnvalue_10 _returnvalue_9))
  (bind _returnvalue_12 (Math|Vector|MakeVector _returnvalue_11 _returnvalue_8 _returnvalue_5))
  (if (Utilities|IsValid _returnvalue_2 :!class /Script/Engine.KismetSystemLibrary)
    ;[0.8694 0.8958 0 0.5] Aiming near a Sideways / Demon Arena bubble was causing GetPlayerAimPoint to fail without checking No Trace
    ; TODO: Why is this function trace failing near these??
    (bind _returnvalue_13 (Fortnite|GetPlayerAimPoint _returnvalue_2 2e+03 (Utilities|Array|MakeArray (CallFunction|GetAvatarasFortPlayerPawn)) true))
    (if (<= (ScalableFloat|GetValueatLevel (Variables|Default|GetRandomizeProjectilesCheck)) 0)
      (return (Math|Rotator|FindLookatRotation Start _returnvalue_13))
      (elif (> (ScalableFloat|GetValueatLevel (Variables|Default|GetDebugEndLocationProjectile)) 0)
        (Rendering|Debug|DrawDebugSphere (+ _returnvalue_13 _returnvalue_12) 2e+01 12 "(R=0.021732,G=0.187500,B=0.181680,A=1.000000)" 5.0 3.0)))
    (return (Math|Rotator|FindLookatRotation Start (+ _returnvalue_13 _returnvalue_12)))
    (else
      (bind _outfortplayerpawn (CallFunction|GetAvatarasValidFortPlayerPawn)
        (:ValidFortPlayerPawn
          (return (Targeting|GetDefaultTargetingRotation _outfortplayerpawn)))
        (:AvatarCastFailed
          (Ability|EndAbility))))))

(fn GetProjectileSpawnLocation
  (bind _verbosedebug (Variables|Default|GetVerboseDebug))
  (bind _outloc (LocalVariables|GetOutLoc))
  (Utilities|FlowControl|Sequence
    (:then_0
      (bind _output_get (LocalVariables|SetOutLoc (+ (Math|Vector|RotateVector (Math|Vector|SelectVector (Variables|Default|GetRidingOffset_Left) (select (Variables|/Script/Engine.Character|GetbIsCrouched (Variables|Default|GetAs%20Fort%20Player%20Pawn)) (Variables|Default|GetCrouchOffset_Left) (Variables|Default|GetStandardOffset)) (GameplayTags|HasMatchingGameplayTag (Variables|Default|GetAs%20Fort%20Player%20Pawn) "(TagName=\"Granted.Riding.Player.IsRiding\")")) (Transformation|GetActorRotation (Variables|Default|GetAs%20Fort%20Player%20Pawn))) (Transformation|GetActorLocation (Variables|Default|GetAs%20Fort%20Player%20Pawn)))))
      (bind _as_fort_player_pawn (Variables|Default|GetAs%20Fort%20Player%20Pawn)
        (:then
          ;[0.573 1 0.8961] Line trace from the actor location to the projectile spawn location to check for any "missed hits". If something is hit, adjust the projectile location so that it is before that hit.
          ; This is necessary since the projectile spawn offset is far enough from the player's location that it can miss hitting something if the player is right up against it.
          (bind (_outhit _returnvalue) (Collision|LineTraceByProfile (Transformation|GetActorLocation _as_fort_player_pawn) _outloc "FortProjectileThrownOverlapPawns" false (Utilities|Array|MakeArray _as_fort_player_pawn (Variables|/Script/FortniteGame.FortPawn|GetCurrentWeapon _as_fort_player_pawn) (Actor|GetAttachParentActor _as_fort_player_pawn)) "None" true "(R=1.000000,G=0.000000,B=0.000000,A=1.000000)" "(R=0.000000,G=1.000000,B=0.000000,A=1.000000)" 1e+01)) ; Do NOT worry about pawns  here. The projectile has a much larger pawn specific collision,\nso this issue does not occur when throwing at a close pawn. 
          (if _returnvalue
            (LocalVariables|SetOutLoc (+ (* (Math|Vector|Normalize (- (Transformation|GetActorLocation _as_fort_player_pawn) _outloc)) (LocalVariables|GetAdjustmentHitOffset)) (.Location (Collision|BreakHitResult _outhit))))
            (Development|PrintString (Utilities|String|Append "Start location trace - Blocking hit, adjusting start location - " (Utilities|GetDisplayName (.HitActor (Collision|BreakHitResult _outhit)))))
            (else
              (Utilities|IsValid (.HitActor (Collision|BreakHitResult _outhit))
                (:Is%20Valid
                  (Development|PrintString (Utilities|String|Append "Start location trace - non blocking hit - " (Utilities|GetDisplayName (.HitActor (Collision|BreakHitResult _outhit)))) _verbosedebug _verbosedebug))
                (:Is%20Not%20Valid
                  (Development|PrintString "Start location trace - No hit" _verbosedebug _verbosedebug))))))
        (:else)))
    (:then_1
      (return (LocalVariables|GetOutLoc)))))
