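; Gameplay ability for the MistRaven ascended flamethrower that plays an outro animation when the player reloads.
; On activation it waits for the player reload gameplay tag to be added, then plays the MistRaven drain outro montage.
; The end-ability event is present but empty, and no montage callbacks are handled.
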(event Ability|EventActivateAbility
  (Ability|Tasks|WaitGameplayTagAdd "(TagName=\"Gameplay.Action.Player.Reload\")"
    (:Added
      (Ability|Tasks|PlayMontageAndWaitNotify "Flamethrower_Outro" "/MistRaven/Animation/MistRaven_Drain_Outro_M.MistRaven_Drain_Outro_M" 1.0 "None" false))))

(event Ability|EventOnEndAbility
  (params (bWasCancelled :type Boolean)))
